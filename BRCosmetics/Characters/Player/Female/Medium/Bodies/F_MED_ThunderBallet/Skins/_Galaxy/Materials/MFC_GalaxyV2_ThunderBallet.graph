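# Material function that adds the optional GalaxyV2 skin effect to the ThunderBallet character, blending it into incoming material attributes by a SkinFX mask.
# Triplanar camera-vector lookups layer a rotating nebula and two star layers, with FOV-compensated tiling and a panning twinkle.
# A Fresnel rim adds noise-driven flame color scaled by the Flashing input.
# Blueprint-fed actor rotation adjusts the camera vector, and a switch makes the emissive additive or overwriting.
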MaterialFunction /BRCosmetics/Characters/Player/Female/Medium/Bodies/F_MED_ThunderBallet/Skins/_Galaxy/Materials/MFC_GalaxyV2_ThunderBallet
graph {nodes [
  n1 = GetMaterialAttributes(None=NamedRerouteUsage()) {attributeGetTypes [
    #1
    #2
    #3
    #4
    #5
   ]}
  n2 = SetMaterialAttributes!(
   MaterialAttributes=@n1.MaterialAttributes
   BaseColor=@n1.MaterialAttributes
   Specular=@n1.MaterialAttributes
   Roughness=@n1.MaterialAttributes
   Normal=@n1.MaterialAttributes
   EmissiveColor=@n1.MaterialAttributes
  ) {attributeSetTypes [
    #4
    #5
    #1
    #2
    #3
   ]}
  n3 = Multiply(A=ScalarParameter() {defaultValue=1, group=GalaxyV2, parameterName=Galaxy_GlobalBrightness, sliderMax=1}, B=NamedRerouteUsage())
  n4 = SetMaterialAttributes(
   BaseColor=Constant()
   Specular=ScalarParameter() {group=GalaxyV2, parameterName=Galaxy_Specular, sliderMax=1}
   Roughness=Multiply(A=ScalarParameter() {defaultValue=1, group=GalaxyV2, parameterName=Galaxy_Roughness, sliderMax=1}, B=@n1.Roughness)
   Normal=@n1.Normal
   EmissiveColor=StaticSwitch(
    True=Add(A=@n1.EmissiveColor, B=@n3)
    False=@n3
    Value=StaticBoolParameter() {desc="false = overwrite emissive", group=GalaxyV2, parameterName=UseAdditiveEmissiveGalaxy, sortPriority=2}
   )
  ) {attributeSetTypes [
    #4
    #5
    #1
    #2
    #3
   ]}
  n5 = NamedRerouteUsage()
  FunctionOutput(None=StaticSwitchParameter(True=BlendMaterialAttributes(A=@n2, B=@n4, Alpha=@n5) {vertexAttributeBlendType=UseA}, False=NamedRerouteUsage()) {group=GalaxyV2, parameterName=UseGalaxyV2, sortPriority=1})
  NamedRerouteDeclaration(None=FunctionInput() {inputName=MA, inputType=FunctionInput_MaterialAttributes}) {name=MA, nodeColor=<rgba 0 0.6484 1 1>}
  NamedRerouteDeclaration(None=ChannelMaskParameter(None=FunctionInput() {inputName=SkinFX, sortPriority=1}) {group=GalaxyV2, parameterName=GalaxyV2_SkinFX_Mask, sortPriority=5}.R) {name=SkinFX_Mask, nodeColor=<rgba 1 0.6562 0 1>}
  n6 = Time()
  n7 = Multiply(A=@n6, B=Constant())
  n8 = ScalarParameter() {group=GalaxyV2, parameterName=BackgroundRotation, sliderMax=1}
  n9 = AppendVector(
   A=ScalarParameter() {defaultValue=0.02, group=GalaxyV2, parameterName=Galaxy_RotationSpeed, sliderMax=1}
   B=ScalarParameter() {defaultValue=-0.02, group=GalaxyV2, parameterName=Galaxy_RotationSpeed_1, sliderMax=1}
  )
  n10 = ScalarParameter() {defaultValue=0.1, group=GalaxyV2, parameterName=Galaxy_RotationSpeed_2, sliderMax=1}
  n11 = TextureObjectParameter() {
   group=GalaxyV2
   parameterName=GalaxyTexture
   sortPriority=3
   texture=/Game/Characters/Player/Male/Medium/Bodies/M_Med_Soldier_04/Skins/Galaxy/Textures/T_FN_Nebula
  }
  n12 = MaterialFunctionCall() {materialFunction=/Engine/Functions/Engine_MaterialFunctions02/Utility/FOV}
  n13 = Constant() {r=0.414}
  n14 = Constant() {r=1.091}
  n15 = Constant()
  n16 = Constant() {r=1}
  n17 = LinearInterpolate(
   A=Constant() {r=3}
   B=Constant() {r=1}
   Alpha=Saturate(None=MaterialFunctionCall(Input=@n12."FOV ", Input Low=@n13, Input High=@n14, Target Low=@n15, Target High=@n16) {materialFunction=&RemapValueRange}.Result)
  )
  n18 = VectorParameter() {group=GalaxyV2, parameterName=Galaxy_Offset, defaultValue=<rgba 0 0 0 1>}
  n19 = MaterialFunctionCall!(
   CameraVector=NamedRerouteUsage()
   Rotation Speed=Multiply(A=LinearInterpolate(A=@n6, B=@n7, Alpha=@n8), B=AppendVector(A=@n9, B=@n10))
   Texture=@n11
   Tiling=Multiply(A=ScalarParameter() {defaultValue=0.5, group=GalaxyV2, parameterName=Galaxy_Tiling, sliderMax=2}, B=@n17)
   Offset=Add(A=ScalarParameter() {defaultValue=0.012, group=GalaxyV2, parameterName=WarpIntensity, sliderMax=1}, B=@n18.RGB)
   AxisFadeContrast=ScalarParameter() {defaultValue=6, group=GalaxyV2, parameterName=Galaxy_AxisFade, sliderMax=8}
  ) {materialFunction=/Game/Weapons/WeaponSkins/MaterialFunction/TriplanarCameraVector_Rotated}
  n20 = Desaturation(None=@n19.Result, Fraction=ScalarParameter() {defaultValue=0.5, group=GalaxyV2, parameterName=StarSaturation, sliderMax=1})
  n21 = Time()
  n22 = ScalarParameter() {defaultValue=-0.125, group=GalaxyV2, parameterName=SmallStar_RotationSpeed, sliderMax=1, sliderMin=-1}
  n23 = TextureObjectParameter() {
   group=GalaxyV2
   parameterName=StarsTexture
   samplerType=SAMPLERTYPE_Masks
   sortPriority=3
   texture=/Game/Characters/Player/Male/Medium/Bodies/M_Med_Soldier_04/Skins/Galaxy/Textures/Celestial_T_Stars
  }
  n24 = ScalarParameter() {defaultValue=5, group=GalaxyV2, parameterName=SmallStar_Tiling, sliderMax=8}
  n25 = ComponentMask(None=MaterialFunctionCall(CameraVector=NamedRerouteUsage(), Rotation Speed=Multiply(A=@n21, B=@n22), Texture=@n23, Tiling=Multiply(A=@n24, B=@n17), Offset=@n18.RGB) {materialFunction=&TriplanarCameraVector_Rotated}.Result) {b=true}
  n26 = LinearInterpolate(
   A=@n19.Result
   B=Multiply(A=@n20, B=VectorParameter() {group=GalaxyV2, parameterName=StarBrightness_Color1, defaultValue=<rgba 300 300 300 1>}.RGB)
   Alpha=Multiply(A=@n25, B=@n25)
  )
  n27 = Multiply(A=@n20, B=VectorParameter() {group=GalaxyV2, parameterName=StarBrightness_Color2, defaultValue=<rgba 950 950 950 1>}.RGB)
  n28 = ScalarParameter() {defaultValue=2, group=GalaxyV2, parameterName=BrightStar_Tiling, sliderMax=4}
  n29 = ComponentMask(None=MaterialFunctionCall!(CameraVector=NamedRerouteUsage(), Texture=@n23, Tiling=Multiply(A=@n28, B=@n17), Offset=@n18.RGB, AxisFadeContrast) {materialFunction=/Engine/Functions/Engine_MaterialFunctions01/Texturing/TriplanarCameraVector}.Result) {r=true}
  n30 = TextureSample(UVs=Panner(Coordinate=TextureCoordinate() {uTiling=2, vTiling=2}) {speedX=0.133, speedY=0.2}, Tex=@n23) {
   samplerSource=SSM_Wrap_WorldGroupSettings
   samplerType=SAMPLERTYPE_Masks
   texture=&Celestial_T_Stars
  }
  n31 = Multiply(A=Multiply(A=@n29, B=@n29), B=@n30.G)
  n32 = ScalarParameter() {defaultValue=8, group=GalaxyV2, parameterName=AccentRimBrightness, sliderMax=1}
  n33 = NamedRerouteUsage()
  n34 = AppendVector(
   A=ScalarParameter() {group=GalaxyV2, parameterName=EdgeFresnel_GradientSpeed_X, sliderMax=2}
   B=ScalarParameter() {defaultValue=0.07, group=GalaxyV2, parameterName=EdgeFresnel_GradientSpeed_Y, sliderMax=2}
  )
  n35 = TextureObject() {samplerType=SAMPLERTYPE_LinearColor, texture=/Game/Weapons/WeaponSkins/Textures/T_Swirly_Noise_highContrast1}
  n36 = ScalarParameter() {defaultValue=1, group=GalaxyV2, parameterName=EdgeFresnel_Tiling, sliderMax=4}
  n37 = TextureCoordinate() {uTiling=2, vTiling=2}
  n38 = TextureSample(UVs=Panner(Coordinate=Multiply(A=@n37, B=@n36), Speed=@n34)) {
   samplerSource=SSM_Wrap_WorldGroupSettings
   samplerType=SAMPLERTYPE_LinearColor
   texture=&T_Swirly_Noise_highContrast1
  }
  n39 = StaticSwitchParameter(
   True=MaterialFunctionCall(Float3=MaterialFunctionCall!(CameraVector=@n33, Time Scale, Panner Speed=@n34, Texture=@n35, Tiling=@n36, Offset, AxisFadeContrast) {materialFunction=/Game/Weapons/WeaponSkins/MaterialFunction/TriplanarCamerVector_Panning}.Result) {materialFunction=&BreakOutFloat3Components}.R
   False=@n38.R
  ) {group=GalaxyV2, parameterName=UseTriplanerFresnelFlame, sortPriority=2}
  n40 = ScalarParameter() {defaultValue=1, group=GalaxyV2, parameterName=Galaxy_FlameFresneIntensity, sliderMax=1}
  n41 = MaterialFunctionCall!(In=LinearInterpolate(
    A=Multiply(A=@n32, B=@n19.Result)
    B=VectorParameter() {group=GalaxyV2, parameterName=EdgeFresnel, defaultValue=<rgba 5.0909 8.586 10 1>}.RGB
    Alpha=Multiply(A=@n39, B=@n40)
   )) {materialFunction=/Game/Packages/Fortress_SharedMaterials/Anime_Master_Material/MF_TODCustomColors}
  n42 = FunctionInput() {inputName=Flashing, inputType=FunctionInput_Scalar}
  n43 = Fresnel(ExponentIn=ScalarParameter() {defaultValue=5, group=GalaxyV2, parameterName=Galaxy_FresEX, sliderMax=8}) {baseReflectFraction=0}
  NamedRerouteDeclaration(None=LinearInterpolate(A=LinearInterpolate(A=@n26, B=@n27, Alpha=@n31), B=Multiply(A=@n41.Result, B=@n42), Alpha=Saturate(None=@n43) {desc=Wormhole Fresnel})) {name=GalaxyEmissive, nodeColor=<rgba 1 0.4219 0 1>}
  n44 = Constant3Vector() {constant=<rgba 0 1 0 0>}
  n45 = Divide(A=Multiply(
    A=VectorParameter() {desc="Dont Touch! this Vector is set by the CCPM BP.", group=GalaxyV2, parameterName=ActorRotationAmount, sortPriority=64}.B
    B=ScalarParameter() {defaultValue=0.1, group=GalaxyV2, parameterName=RotationAmount, sliderMax=1}
   ) {constB=0.333}) {constB=360}
  n46 = Constant3Vector()
  n47 = Transform(None=CameraVectorWS()) {transformSourceType=TRANSFORMSOURCE_World, transformType=TRANSFORM_Camera}
  n48 = MaterialFunctionCall(Float3=Add(A=RotateAboutAxis(NormalizedRotationAxis=@n44, RotationAngle=@n45, PivotPoint=@n46, Position=@n47), B=@n47)) {materialFunction=&BreakOutFloat3Components}
  NamedRerouteDeclaration(None=MaterialFunctionCall(X=@n48.B, Y=@n48.B, Z=@n48.B) {materialFunction=&MakeFloat3}.Result) {name=CameraVector, nodeColor=<rgba 1 0.0833 0.3367 1>}
  TransformPosition(None=@n46) {transformType=TRANSFORMPOSSOURCE_World}
 ]}
