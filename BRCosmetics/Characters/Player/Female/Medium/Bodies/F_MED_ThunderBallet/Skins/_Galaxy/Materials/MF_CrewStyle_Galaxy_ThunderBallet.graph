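# Material function that applies the Galaxy crew style to the ThunderBallet character skin, toggled by the Galaxy static switch.
# It runs the input attributes through MFC_GalaxyV2_ThunderBallet, masked by a selectable Tech Art Mask channel, then adds a tinted emissive glow.
# The glow combines an inverted height gradient for the legs and a spherical core, both from pre-skinned position, shaped by an occlusion-style term from the Mask input.
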MaterialFunction /BRCosmetics/Characters/Player/Female/Medium/Bodies/F_MED_ThunderBallet/Skins/_Galaxy/Materials/MF_CrewStyle_Galaxy_ThunderBallet
graph {nodes [
  n1 = ChannelMaskParameter(None=FunctionInput() {inputName=Tech Art Mask}) {parameterName=GalaxyChannel}
  n2 = MaterialFunctionCall(Float3=FunctionInput() {inputName=Mask}) {materialFunction=&BreakOutFloat3Components}
  n3 = Multiply(A=@n2.R, B=Multiply(A=@n2.G, B=OneMinus(None=@n2.B)))
  n4 = ScalarParameter() {defaultValue=1.5, group=Galaxy Bonus Gradient, parameterName=GalaxyAOIntensity}
  n5 = Multiply(A=@n1.R, B=Saturate(None=Power(Base=@n3, Exp=@n4) {constExponent=1.5}))
  Add(A=@n5) {constB=0.05}
  n6 = FunctionInput() {inputType=FunctionInput_MaterialAttributes}
  n7 = MaterialFunctionCall!(Flashing=FunctionInput() {inputName=Flashing, inputType=FunctionInput_Scalar}, MA=@n6, SkinFX=@n1.R) {materialFunction=&/MFC_GalaxyV2_ThunderBallet}
  n8 = VectorParameter() {group=Galaxy Bonus Gradient, parameterName=Galaxy Emissive Gradient Color, defaultValue=<rgba 0 0.8408 1 1>}
  n9 = VertexInterpolator(VS=PreSkinnedPosition())
  n10 = Divide(
   A=Subtract(A=@n9.PS, B=ScalarParameter() {group=Galaxy Bonus Gradient, parameterName=GradientGlowoffset})
   B=LinearInterpolate(Alpha=ScalarParameter() {defaultValue=0.5, group=Galaxy Bonus Gradient, parameterName=GalaxyEmissiveHeightAdjustment}) {constA=75, constB=140}
  ) {constB=120}
  n11 = Clamp(None=MaterialFunctionCall(Float3=OneMinus(None=@n10)) {materialFunction=&BreakOutFloat3Components}.B)
  n12 = ScalarParameter() {defaultValue=0.2, group=Galaxy Bonus Gradient, parameterName=LegGradientIntensity}
  n13 = SphereMask(
   A=@n9.PS
   B=VectorParameter() {parameterName=pos, defaultValue=<rgba 0 0 125 1>}.RGB
   Radius=ScalarParameter() {defaultValue=30, group=Galaxy Bonus Gradient, parameterName=CoreGlowRadius}
  ) {hardnessPercent=0}
  n14 = Add(A=Multiply(A=Saturate(None=@n5), B=Multiply(A=@n11, B=@n12)), B=Multiply(A=Power(Base=@n13) {constExponent=3}, B=@n5))
  n15 = ScalarParameter() {defaultValue=4, group=Galaxy Bonus Gradient, parameterName=GalaxyEmissiveGradientOpacity}
  n16 = Add(
   A=GetMaterialAttributes(None=@n7.Result) {attributeGetTypes [#1]}.EmissiveColor
   B=Multiply(A=@n8.RGB, B=Multiply(A=@n14, B=@n15))
  )
  FunctionOutput(None=StaticSwitchParameter(
    True=SetMaterialAttributes!(MaterialAttributes=@n7.Result, EmissiveColor=@n16) {attributeSetTypes [#1]}
    False=@n6
   ) {parameterName=Galaxy})
  MaterialFunctionCall!(In=@n7.Result, Warp Mask, GalaxyAlternateFresnel, Alpha=@n1.R) {materialFunction=/Game/Weapons/WeaponSkins/MaterialFunction/Galaxy}
 ]}
